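# Masked lit Niagara sprite and mesh material for the MarineCarve skydiving contrail, drawing either icons or skateboard pieces.
# The particle's Progress dynamic parameter fades color from desaturated to full as it passes 0.4.
# A SkateboardPieces static switch selects opaque pieces using the backpack's diffuse, normal and mask textures.
# Otherwise icons use emissive-adjusted color, flat surface values and a temporally dithered mask from their red and blue channels.
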Material /BRCosmetics/Athena/Effects/Particles/Skydiving/NS_Contrail_MarineCarve/M_Contrail_MarineCarve
graph {
 properties {
  bAutomaticallySetUsageInEditor=false
  bHasPixelAnimation=true
  bUsedWithNiagaraMeshParticles=true
  bUsedWithNiagaraSprites=true
  blendMode=BLEND_Masked
  usedShadingModels=Default Lit
 }
 nodes [
  n1 = TextureSample() {texture=&/T_Contrail_MarineCarve_Icons}
  n2 = DynamicParameter() {paramNames [Progress,Param2,Param3,Param4]}
  n3 = InverseLinearInterpolate(Value=@n2.Progress) {constB=0.2}
  n4 = Subtract(A=@n1.A, B=OneMinus(None=Saturate(None=@n3)))
  n5 = SmoothStep(Value=Add(A=Add(A=@n1.R, B=@n1.R), B=@n1.B)) {constMax=0.1}
  n6 = Saturate(None=InverseLinearInterpolate(Value=@n2.Progress) {constA=0.4})
  LinearInterpolate(A=@n4, B=@n5, Alpha=@n6)
  LinearInterpolate(A=@n4, B=Multiply(A=@n1.RGB, B=@n4), Alpha=Saturate(None=InverseLinearInterpolate(Value=@n2.Progress) {constA=0.2, constB=0.4}))
  ScalarParameter() {parameterName=Progress}
  n7 = TextureSample() {texture=/BRCosmetics/Accessories/FORT_Backpacks/Backpack_F_MED_MarineCarve/Textures/T_MarineCarve_Backpack_D}
  n8 = Constant() {r=0.9}
  n9 = LinearInterpolate(A=Desaturation(None=@n1.RGB), B=@n1.RGB, Alpha=@n6)
  StaticSwitchParameter(
   True=LinearInterpolate(A=Desaturation(None=@n7.RGB, Fraction=@n8), B=@n7.RGB, Alpha=@n6)
   False=Multiply(A=MaterialFunctionCall(Emissive=@n9) {materialFunction=&MF_FX_EmissiveAdjustments}.Result) {constB=0.5}
  ) {parameterName=SkateboardPieces}
  n10 = TextureSample() {samplerType=SAMPLERTYPE_Normal, texture=/BRCosmetics/Accessories/FORT_Backpacks/Backpack_F_MED_MarineCarve/Textures/T_MarineCarve_Backpack_N}
  StaticSwitchParameter(True=@n10.RGB, False=Constant3Vector() {constant=<rgba 0 0 1 0>}) {parameterName=SkateboardPieces}
  StaticSwitchParameter(
   True=Constant() {r=1}
   False=MaterialFunctionCall(Alpha Threshold=@n5) {materialFunction=&DitherTemporalAA}.Result
  ) {parameterName=SkateboardPieces}
  n11 = TextureSample() {samplerType=SAMPLERTYPE_Masks, texture=/BRCosmetics/Accessories/FORT_Backpacks/Backpack_F_MED_MarineCarve/Textures/T_MarineCarve_Backpack_S}
  StaticSwitchParameter(True=@n11.R, False=Constant() {r=0.5}) {parameterName=SkateboardPieces}
  StaticSwitchParameter(True=@n11.G, False=Constant() {r=0.5}) {parameterName=SkateboardPieces}
  StaticSwitchParameter(True=@n11.B, False=Constant()) {parameterName=SkateboardPieces}
 ]
}
